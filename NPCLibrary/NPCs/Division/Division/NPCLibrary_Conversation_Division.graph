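# Conversation database for the Division NPC in NPCLibrary, defining the greeting and the branches a player gets when talking to it.
# The entry point leads to one Speech node with a default line and contextual lines for returning players, first meetings, high Bar totals and a specific outfit.
# Each contextual line pairs a controller requirement with a curve-driven priority.
# The speech then branches through link nodes into the shared NPCLibrary Quests and BasicServices conversations.
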ConversationDatabase /NPCLibrary/NPCs/Division/Division/NPCLibrary_Conversation_Division
entry_points [{tag=Conversation.Entry.Division, destinations [#1]}]
nodes [
 ConversationEntryPointNode_7 {
  id=#1
  class=ConversationEntryPointNode
  outputs [#2]
  properties {entryTag {tagName=Conversation.Entry.Division}}
 }
 Speech_C_4 {
  id=#2
  class=Speech_C
  outputs [#3, #4]
  properties {
   nodeName=""
   generalConfig {
    defaultMessage=You need some work done?
    contextualMessages [
     {
      message=Good to have you back. There's a situation ongoing.
      contextRequirements [{requirement {class=FortControllerRequirement_CheckTimeSinceLastPlayedMatch}}]
      priority {curve {rowName=Tandem.Speech.Priority.WhileSincePlayed}}
     }
     {
      message=If you're looking for a mercenary... I may know someone.
      contextRequirements [{requirement {class=FortControllerRequirement_IsFirstEverConversationWithNPC}}]
      priority {curve {rowName=Tandem.Speech.Priority.FirstEverInteraction}}
     }
     {
      message=I'll take those Bars if you don't want 'em.
      contextRequirements [{requirement {class=FortControllerRequirement_HasBars, properties {thresholdTestConfig {quantity {value=4000}}}}}]
      priority {curve {rowName=Tandem.Speech.Priority.HighTotalBars}}
     }
     {
      message=I thought we said we'd only meet in private from now on.
      contextRequirements [{requirement {
         class=FortControllerRequirement_HasCID
         properties {allowedCIds [/BRCosmetics/Athena/Items/Cosmetics/Characters/CID_A_113_Athena_Commando_F_Innovator]}
        }}]
      priority {curve {rowName=Tandem.Speech.Priority.OutfitReaction}}
     }
    ]
   }
  }
 }
 ConversationLinkNode_0 {id=#4, class=ConversationLinkNode, properties {remoteEntryTag {tagName=Conversation.Entry.BasicServices}}}
 ConversationLinkNode_1 {id=#3, class=ConversationLinkNode, properties {remoteEntryTag {tagName=Conversation.Entry.NPCLibrary.Quests}}}
]
